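# Niagara idle effect for the BugBandit character skin, adding breath smoke and glowing red eyes to the skeletal mesh.
# BreathSmoke spawns orange-to-brown puffs from a loop-age rate curve, with erosion-style dynamic material parameters and rising forces.
# EyeGlow partitions two sprites onto separate sockets and feeds offset fake-light parameters alongside a second renderer.
# Head-hidden state suppresses both, while big-head mode, platform, time of day and front end adjust size or brightness.
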NiagaraSystem /BRCosmetics/Characters/Player/Male/Medium/Bodies/M_MED_Moth/Skins/BugBandit/FX/NS_Character_BugBandit_Idle
dependencies {
 usedDataInterfaces [
  &NiagaraDataInterfaceCurve
  &NiagaraDataInterfacePlatformSet
  &NiagaraDataInterfaceColorCurve
  &NiagaraDataInterfaceSpriteRendererInfo
  &NiagaraDataInterfaceArrayInt32
  &NiagaraDataInterfaceArrayFloat3
  &NiagaraDataInterfaceSkeletalMesh
 ]
 usedDynamicInputs [&Multiply_FloatByInt, &ReturnParticleIDAttributes]
 usedModules [
  &self:SystemSpawnScript.SystemScriptSource.SystemScriptGraph.NiagaraNodeAssignment_0.SetVariables_#1
  &SystemState
 ]
}
emitters {
 BreathSmoke {
  emitterUpdateScript [
   &EmitterState()
   &SpawnRate(
    Spawn Probability=&MakeCustomFloatFromBool(Bool=$User.bIsHeadHidden, True Float=0, False Float=1)
    SpawnRate=&FloatFromCurve(
     FloatCurve=DataInterface() {propertyValues {Curve {
        defaultValue=1
        keys [
         {arriveTangent=-0.1347, interpMode=RCIM_Cubic, leaveTangent=-0.1347, tangentMode=RCTM_User, time=0.2, value=1}
         {arriveTangent=-0.092, interpMode=RCIM_Cubic, leaveTangent=-0.092, tangentMode=RCTM_User, time=0.2319, value=0}
        ]
       }}}
     CurveIndex=$Emitter.NormalizedLoopAge
     Scale Curve=12
    )
    Use Spawn Probability=true
   )
   &SpawnBurst_Instantaneous(
    Spawn Probability=&MakeCustomFloatFromBool(Bool=$User.bIsHeadHidden, True Float=0, False Float=1)
    Use Spawn Probability=true
    UseLoopCountLimit=false
   )
  ]
  particleSpawnScript [
   &InitializeParticle(
    Color Mode=::NewEnumerator0
    Lifetime Mode=::NewEnumerator1
    Sprite Rotation Mode=::NewEnumerator2
    Sprite Size Mode=::NewEnumerator3
    Uniform Sprite Size=&Lerp_Float(A=25, B=45, Alpha=$NPC.FortniteNPC.BigHeadMode)
    UsePositionOffset=false
    Write Material Random=true
   )
   &AddVelocity(Velocity=&RandomRangeVector(Minimum=<0 0 -20>, Maximum=<0 0 20>))
   &CameraOffset(Camera Offset Amount=&FN_PickFloatDependOnPlatform(TargetPlatformActive=0, TargetPlatformNoActive=5))
  ]
  particleUpdateScript [
   &ParticleState()
   &SkeletalMeshLocation(MaskSpawnGroup=false, Mesh Sampling Type=::NewEnumerator5, Skeletal Mesh=$System.Skeletal Mesh Reader, Socket Sampling Mode=::NewEnumerator5)
   &ScaleSpriteSize(
    Scale Sprite Size Mode=::NewEnumerator0
    Uniform Scale Factor=&RandomRangeFloat(
     Minimum=&FloatFromCurve(FloatCurve=DataInterface() {propertyValues {Curve {
         defaultValue=1
         keys [
          {arriveTangent=0.9695, interpMode=RCIM_Cubic, leaveTangent=0.9695, tangentMode=RCTM_User, time=0, value=0}
          {interpMode=RCIM_Cubic, tangentMode=RCTM_User, time=1, value=1}
         ]
        }}})
     Maximum=&FloatFromCurve(FloatCurve=DataInterface() {propertyValues {Curve {
         defaultValue=1
         keys [
          {arriveTangent=3.8646, interpMode=RCIM_Cubic, leaveTangent=3.8646, tangentMode=RCTM_User, time=0, value=0}
          {interpMode=RCIM_Cubic, tangentMode=RCTM_User, time=1, value=3}
         ]
        }}})
    )
   ) {module=ScaleSpriteSize001}
   &DynamicMaterialParameters(
    Index 0 Param 1=&FloatFromCurve(FloatCurve=DataInterface() {propertyValues {Curve {defaultValue=1, keys [{arriveTangent=1.3169, time=0, value=0.3}, {arriveTangent=1.3666, leaveTangent=1.3666, time=1, value=1.5}]}}})
    Index 0 Param 2=&FloatFromCurve(FloatCurve=DataInterface() {propertyValues {Curve {
        defaultValue=1
        keys [
         {arriveTangent=0.4074, interpMode=RCIM_Cubic, leaveTangent=0.4074, tangentMode=RCTM_User, time=0, value=0.08}
         {interpMode=RCIM_Cubic, tangentMode=RCTM_User, time=0.5, value=0.3}
        ]
       }}})
    Index 0 Param 3=1
    Module.Index0Float0Write=true
    Module.Index0Float1Write=true
    Module.Index0Float2Write=true
    Module.Index0Float3Write=true
   )
   &ScaleColor(
    Scale Alpha=&FloatFromCurve(
     FloatCurve=DataInterface() {propertyValues {Curve {
        defaultValue=1
        keys [
         {interpMode=RCIM_Cubic, tangentMode=RCTM_User, time=0, value=0}
         {arriveTangent=1.75, interpMode=RCIM_Cubic, leaveTangent=1.75, time=0.2, value=1}
         {arriveTangent=-2.7575, interpMode=RCIM_Cubic, leaveTangent=-2.7575, tangentMode=RCTM_User, time=0.4, value=0.7}
         {arriveTangent=-1.1667, interpMode=RCIM_Cubic, leaveTangent=-1.1667, time=0.7, value=0.1}
         {interpMode=RCIM_Cubic, tangentMode=RCTM_User, time=1, value=0}
        ]
       }}}
     Scale Curve=&FN_PickFloatDependOnPlatform(TargetPlatformActive=1.5, TargetPlatformNoActive=1.2)
    )
    Scale RGB=&MakeVectorFromLinearColorRGB(Color=&ColorFromCurve(ColorCurve=DataInterface() {propertyValues {
        AlphaCurve {keys [<at 0 1 RCIM_Cubic>,<at 1 0 RCIM_Cubic>]}
        BlueCurve {keys [<at 0 0.1406 RCIM_Cubic>,<at 0.7664 0.0275 RCIM_Cubic>]}
        GreenCurve {keys [<at 0 0.4545 RCIM_Cubic>,<at 0.7664 0.103 RCIM_Cubic>]}
        RedCurve {keys [<at 0 1 RCIM_Cubic>,<at 0.7664 0.3854 RCIM_Cubic>]}
       }}))
    ScaleA=true
    ScaleRGB=true
   )
   &SubUVAnimation(
    "Sprite Renderer "=DataInterface() {propertyValues {SpriteRenderer=&self:BreathSmoke_0.NiagaraSpriteRendererProperties_0}}
    SubUV Lookup Index Scale=0
    UseEndFrame=false
    UseStartFrame=false
   )
   &Drag(Drag=1.5, Use Linear Drag=true, Use Rotational Drag=true)
   &AccelerationForce(Acceleration=&RandomRangeVector(Minimum=<0 0 18>, Maximum=<1 1 24>))
   &GravityForce(Gravity=<0 0 700>)
   &SolveForcesAndVelocity(Clamp Velocity=false, Limit Acceleration=false)
  ]
  rendererClasses [&NiagaraSpriteRendererProperties]
  renderers [{rendererClass=&NiagaraSpriteRendererProperties}]
 }
 EyeGlow {
  emitterUpdateScript [
   &EmitterState(
    Inactive Response=::NewEnumerator1
    Life Cycle Mode=::NewEnumerator1
    Loop Behavior=::NewEnumerator1
    Loop Duration Mode=::NewEnumerator1
    UseLoopDelay=false
   )
   &SpawnBurst_Instantaneous(
    Spawn Count=2
    Spawn Probability=&MakeCustomFloatFromBool(Bool=$User.bIsHeadHidden, True Float=0, False Float=1)
    Use Spawn Probability=true
    UseLoopCountLimit=false
   )
  ]
  particleSpawnScript [
   &PartitionParticles()
   &InitializeParticle(
    Color=&Multiply_LinearColorByFloat(
     Linear Color=&Multiply_LinearColorByFloat(
      Linear Color=<rgba 3 0 0 1>
      Float=&FN_PickFloatDependOnPlatform(TargetPlatformActive=1, TargetPlatformNoActive=0.6)
     )
     Float=&Multiply_Float(
      A=&FNTimeOfDayScalar(Night Scale=0.8, Morning Scale=0.8, Day Scale=0.8, Evening Scale=0.8)
      B=&FN_FrontEndFloatModifier(Gameplay=1, Front End=0.4)
     )
    )
    Lifetime=4
    Mesh Scale=<1 0.8 0.8>
    Mesh Scale Mode=::NewEnumerator4
    Sprite Size Mode=::NewEnumerator3
    Uniform Sprite Size=22
    UsePositionOffset=false
    Write Material Random=true
   )
  ]
  particleUpdateScript [
   &ParticleState(Kill Particles When Lifetime Has Elapsed=false, Let Infinitely Lived Particles Die When Emitter Deactivates=true)
   &ScaleSpriteSize(
    Scale Sprite Size Mode=::NewEnumerator0
    Uniform Scale Factor=&Lerp_Float(A=1, B=2, Alpha=$NPC.FortniteNPC.BigHeadMode)
   )
   &SkeletalMeshLocation(
    Bone / Socket Index=&SelectIntFromArray(Direct Array Index=$Particles.PartitionParticles.Partition, Int Selection Array=DataInterface() {propertyValues {IntData [1,2]}})
    MaskSpawnGroup=false
    Mesh Sampling Type=::NewEnumerator5
    Orientation Sampling=::NewEnumerator1
    Sampled Position Offset Coordinate Space:ENiagaraSkelMeshTransforms=::NewEnumerator1
    Skeletal Mesh=$System.Skeletal Mesh Reader
    Socket Sampling Mode=::NewEnumerator5
   ) {module=SkeletalMeshLocation001}
   /SetVariables(
    Particles.ColorFakeLight:LinearColor=&Multiply_LinearColorByFloat(
     Linear Color=&Multiply_LinearColorByFloat(
      Linear Color=<rgba 1 0 0.0139 1>
      Float=&FN_PickFloatDependOnPlatform(TargetPlatformActive=1, TargetPlatformNoActive=0.4)
     )
     Float=&Multiply_Float(
      A=&FNTimeOfDayScalar(Night Scale=0.8, Morning Scale=1, Evening Scale=1)
      B=&MakeCustomFloatFromBool(Bool=$User.bIsHeadHidden, True Float=0, False Float=0.8)
     )
    )
    Particles.FakeLightPosition:NiagaraPosition=&AddVectorToPosition(
     Position=$Particles.SkeletalMeshLocation001.SampledPosition
     Vector=&SelectVectorFromArray(
      Direct Array Index=$Particles.PartitionParticles.Partition
      Vector Selection Array=DataInterface() {propertyValues {FloatData [<3 0 0>, <-3 0 0>]}}
     )
    )
    Particles.FakeLightSize:Vector2f=&Vector2DFromFloat(Value=20)
   ) {isSetParameters=true}
   &ScaleColor(
    Scale Alpha=&MakeCustomFloatFromBool(Bool=$User.bIsHeadHidden, True Float=0, False Float=1)
    ScaleA=true
    ScaleRGB=true
   )
   &CameraOffset()
  ]
  rendererClasses [&NiagaraSpriteRendererProperties]
  renderers [
   {rendererClass=&NiagaraSpriteRendererProperties}
   {rendererClass=&NiagaraSpriteRendererProperties, rendererIndex=1}
  ]
 }
}
summary {
 systemName=NS_Character_BugBandit_Idle
 emitters [
  {emitterName=BreathSmoke, rendererClasses [&NiagaraSpriteRendererProperties]}
  {emitterName=EyeGlow, rendererClasses [&NiagaraSpriteRendererProperties]}
 ]
 userVariables [
  User.bIsFrontEnd {type=NiagaraBool}
  User.bIsFrontEndPreview {type=NiagaraBool}
  User.bIsHeadHidden {type=NiagaraBool}
  User.BlinkAlpha {type=NiagaraFloat, defaultValue=0, description=""}
 ]
}
